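# Sound cue for a looping heist lockdown music track, kept in a prototype music folder.
# A single looping wave player plays the mastered track on the gameplay music sound class, non-spatialized and with boosted volume.
# Overridden concurrency allows only one instance and blocks new ones, and the cue keeps playing while silent through virtualization.
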SoundCue /Clyde/Audio/PrototypeMusic_DNU/Mx_Heist_Lockdown_LP_Cue
root=WavePlayer_1
nodes [{id=WavePlayer_1, properties {bLooping=true, soundWaveAssetPtr=/Clyde/Audio/PrototypeMusic_DNU/wav/mx_Heist_Lockdown_LP_Mastered}}]
properties {
 bOverrideConcurrency=true
 maxDistance=2097152
 priority=2
 soundClassObject=/Game/Sounds/FortSoundClassesAndModes/Fort_Music_Gameplay
 virtualizationMode=PlayWhenSilent
 volumeMultiplier=1.734
 attenuationOverrides {bSpatialize=false}
 concurrencyOverrides {maxCount=1, resolutionRule=PreventNew}
 concurrencySet [null]
}
