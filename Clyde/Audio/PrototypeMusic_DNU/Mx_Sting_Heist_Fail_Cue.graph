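# Sound cue for the heist failure music stinger, kept in a prototype music folder.
# A Random node picks with equal weight among five heist fail sting wave variants.
# It routes to the music stinger sound class with boosted volume and spatialization disabled in its attenuation overrides.
# Overridden concurrency allows one instance at a time, blocking new plays, with a retrigger time to prevent stacking.
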SoundCue /Clyde/Audio/PrototypeMusic_DNU/Mx_Sting_Heist_Fail_Cue
root=Random_1
nodes [
 {id=Random_1, children [WavePlayer_0,WavePlayer_1,WavePlayer_7,WavePlayer_8,WavePlayer_9], properties {weights [1,1,1,1,1]}}
 {id=WavePlayer_0, properties {soundWaveAssetPtr=/Clyde/Audio/PrototypeMusic_DNU/wav/mx_Heist_Sting_Fail_01}}
 {id=WavePlayer_1, properties {soundWaveAssetPtr=/Clyde/Audio/PrototypeMusic_DNU/wav/mx_Heist_Sting_Fail_02}}
 {id=WavePlayer_7, properties {soundWaveAssetPtr=/Clyde/Audio/PrototypeMusic_DNU/wav/mx_Heist_Sting_Fail_03}}
 {id=WavePlayer_8, properties {soundWaveAssetPtr=/Clyde/Audio/PrototypeMusic_DNU/wav/mx_Heist_Sting_Fail_04}}
 {id=WavePlayer_9, properties {soundWaveAssetPtr=/Clyde/Audio/PrototypeMusic_DNU/wav/mx_Heist_Sting_Fail_05}}
]
properties {
 bOverrideConcurrency=true
 maxDistance=2097152
 soundClassObject=/Game/Sounds/FortSoundClassesAndModes/Fort_Music_Stinger_PSM
 volumeMultiplier=2.5
 attenuationOverrides {bSpatialize=false}
 concurrencyOverrides {maxCount=1, resolutionRule=PreventNew, retriggerTime=3}
}
