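# Sound cue for a looping security camera alarm, stored in Clyde's PrototypeMusic_DNU folder, that reuses the distant idle alarm loop from the SecurityTurret gadget sounds.
# A looping wave player feeds a Modulator that applies a fixed volume cut with no pitch or volume randomization.
# The cue routes to the gadgets sound class, spatializes through the shared alarm light attenuation asset and overrides concurrency to cap simultaneous instances.
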SoundCue /Clyde/Audio/PrototypeMusic_DNU/Security_Camera_Clyde_Alarm_Loop_Cue
root=Modulator_0
nodes [
 {id=Modulator_0, children [WavePlayer_4], properties {pitchMax=1, pitchMin=1, volumeMax=0.7, volumeMin=0.7}}
 {
  id=WavePlayer_4
  properties {bLooping=true, soundWaveAssetPtr=/Game/Athena/Sounds/Gadgets/SecurityTurret/Security_Camera_Alarm_Idle_Loop_02_Distant}
 }
]
properties {
 attenuationSettings=/FNGameplayCues/Athena/Sounds/Gadgets/SecurityTurret/Security_Camera_Alarm_Light_Att
 bOverrideConcurrency=true
 maxDistance=6000
 soundClassObject=&Fort_Gadgets
 volumeMultiplier=1.2
 attenuationOverrides {distanceAlgorithm=Custom, falloffDistance=10112.3027, attenuationShapeExtents {x=600}}
 concurrencyOverrides {maxCount=3}
}
